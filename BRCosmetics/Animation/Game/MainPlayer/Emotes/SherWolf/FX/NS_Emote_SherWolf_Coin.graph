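# Niagara effect for the SherWolf emote that showers spinning coins and golden glints from overhead, tapering off by three seconds.
# Coin spawns meshes with cone and point velocities, drops them under randomized gravity and bounces them, with collision gated by platform quality and a random check.
# Flare runs alongside on a matching spawn curve, drawing falling twinkling sprites through two renderers, scaled down when User.bIsFrontEndPreview is set.
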NiagaraSystem /BRCosmetics/Animation/Game/MainPlayer/Emotes/SherWolf/FX/NS_Emote_SherWolf_Coin
dependencies {
 usedDataInterfaces [&NiagaraDataInterfaceCurve, &NiagaraDataInterfacePlatformSet]
 usedDynamicInputs []
 usedModules [&SystemState]
}
emitters {
 Coin {
  emitterUpdateScript [
   &EmitterState(Inactive Response=::NewEnumerator1, Life Cycle Mode=::NewEnumerator1, Loop Behavior=::NewEnumerator1, Loop Duration=6, UseLoopDelay=false)
   &SpawnRate(
    SpawnRate=&FloatFromCurve(FloatCurve=DataInterface() {propertyValues {Curve {keys [<at 0 1>,<at 2.5 1>,<at 3 0>]}}}, CurveIndex=$Emitter.Age, Scale Curve=10)
    Use Spawn Probability=false
   )
  ]
  particleSpawnScript [
   &InitializeParticle(
    Lifetime Max=1.8
    Lifetime Min=1.3
    Lifetime Mode=::NewEnumerator1
    Mesh Scale Mode=::NewEnumerator3
    Mesh Uniform Scale=1.8
    Sprite Size Mode=::NewEnumerator1
    Uniform Sprite Size Max=25
    Uniform Sprite Size Min=15
    UsePositionOffset=false
    Write Material Random=true
   )
   &ShapeLocation(MaskSpawnGroup=false, Offset=<0 0 250>, Offset Mode=::NewEnumerator0, Override Seed=false, Sphere Radius=30, UseOwnerScale=true)
   &InitialMeshOrientation(Use Rotation Vector=false)
   &AddVelocity(
    Cone Angle=20
    Cone Axis=<0 0 1>
    Distribution Along Cone Axis=0.25
    Module.UseConeFalloff=true
    Module.UseVelDistribution=true
    Override Seed=false
    Speed Falloff From Cone Axis=0.5
    Velocity Mode=::NewEnumerator2
    Velocity Speed=&RandomRangeFloat(Minimum=20, Maximum=30)
   )
   &AddVelocity(Module.Use Falloff Distance=false, Origin Offset=<0 0 260>, Override Seed=false, Velocity Mode=::NewEnumerator1, Velocity Speed=10) {module=AddVelocity001}
   /SetVariables(Particles.CameraOffset002=5) {isSetParameters=true}
  ]
  particleUpdateScript [
   &ParticleState()
   &ScaleSpriteSize(Uniform Curve Sprite Scale=DataInterface() {propertyValues {Curve {keys [<at 0 1>, <at 0.85 1>, {tangentMode=RCTM_User, time=1, value=0.4}]}}})
   &ScaleMeshSize(Scale Factor=&Multiply_VectorByFloat(
     Vector=<1 1 1>
     Float=&FloatFromCurve(FloatCurve=DataInterface() {propertyValues {Curve {keys [<at 0 0>,<at 0.1 1>,<at 0.8 1>,<at 1 0>]}}})
    ))
   &UpdateMeshOrientation(Rotation Rate=&RandomRangeFloat(Minimum=1, Maximum=5))
   &GravityForce(Gravity=&RandomRangeVector(Minimum=<50 50 -500>, Maximum=<-50 -50 -700>))
   &Collision(
    Age Colliding Particles=false
    CPU Collision Trace Channel=ECC_WorldStatic
    Collision Enabled=&Boolean_AndOperation(
     Bool1=&IsPlatformSetActive(New Platform Set=DataInterface() {propertyValues {Platforms {qualityLevelMask=-4}}})
     Bool2=&SetBoolByIntComparison(A=&RandomRangeInt(Minimum=0, Maximum=1), B=0, Comparison Type=::NewEnumerator4)
    )
    Control Roll On Collision=true
    Friction=1
    Friction During a Bounce=0.05
    KillOnCollision=false
    Module.Control Rest State Via Surface Normal=false
    Module.EnableMaxCollisionCount=false
    Module.Randomize Collision Normal Vector=false
    Restitution=0.35
    ScaleRateOfRotation=false
   )
   &SolveForcesAndVelocity(Clamp Velocity=false, Limit Acceleration=false)
  ]
  rendererClasses [&NiagaraMeshRendererProperties]
  renderers [{rendererClass=&NiagaraMeshRendererProperties, rendererIndex=0}]
 }
 Flare {
  emitterUpdateScript [
   &EmitterState(Inactive Response=::NewEnumerator1, Life Cycle Mode=::NewEnumerator1, Loop Behavior=::NewEnumerator1, UseLoopDelay=false)
   &SpawnRate(
    SpawnRate=&FloatFromCurve(FloatCurve=DataInterface() {propertyValues {Curve {keys [<at 0 1>,<at 2.5 1>,<at 3 0>]}}}, CurveIndex=$Emitter.Age, Scale Curve=10)
    Use Spawn Probability=false
   )
  ]
  particleSpawnScript [
   &InitializeParticle(
    Color=<rgba 3 1.5638 0.1236 1>
    Lifetime Max=1.5
    Lifetime Mode=::NewEnumerator1
    Sprite Size Mode=::NewEnumerator1
    Uniform Sprite Size Max=30
    Uniform Sprite Size Min=20
    UsePositionOffset=false
    Write Material Random=true
   )
   &ShapeLocation(MaskSpawnGroup=false, Offset=<0 0 240>, Offset Mode=::NewEnumerator0, Override Seed=false, Sphere Radius=20, UseOwnerScale=true)
   &AddVelocity(Velocity=<0 0 -10>)
  ]
  particleUpdateScript [
   &ParticleState()
   &ScaleSpriteSize(
    Uniform Curve Scale=&Multiply_Float(A=1, B=&MakeCustomFloatFromBool(Bool=$User.bIsFrontEndPreview, True Float=0.8, False Float=1))
    Uniform Curve Sprite Scale=DataInterface() {propertyValues {Curve {keys [
        {tangentMode=RCTM_User, time=0, value=0}
        <at 0.2 0.55>
        <at 0.3 0.25>
        <at 0.4 0.85>
        <at 0.6 0.2>
        <at 0.8 0.975>
        {tangentMode=RCTM_User, time=1, value=0.35}
       ]}}}
   )
   &GravityForce(Gravity=&RandomRangeVector(Minimum=<-30 -30 -200>, Maximum=<30 30 -400>))
   &SolveForcesAndVelocity(Clamp Velocity=false, Limit Acceleration=false)
   &ScaleColor(
    Scale Alpha=&FloatFromCurve(FloatCurve=DataInterface() {propertyValues {Curve {keys [<at 0 0>,<at 0.1 1>,<at 0.9 1>,<at 1 0>]}}})
    ScaleA=true
    ScaleRGB=true
   )
   /SetVariables(Particles.Color002=&MakeLinearColorFromVectorAndFloat(
     "Vector (RGB)"=&MakeVectorFromLinearColorRGB(Color=<rgba 2 1.171 0.0974 1>)
     "Float (Alpha)"=&FloatFromCurve(FloatCurve=DataInterface() {propertyValues {Curve {keys [<at 0 0>,<at 0.1 1>,<at 0.8 1>,<at 1 0>]}}}, Scale Curve=0.05)
    )) {isSetParameters=true}
  ]
  rendererClasses [&NiagaraSpriteRendererProperties]
  renderers [
   {rendererClass=&NiagaraSpriteRendererProperties}
   {rendererClass=&NiagaraSpriteRendererProperties, rendererIndex=1}
  ]
 }
}
summary {
 systemName=NS_Emote_SherWolf_Coin
 emitters [
  {emitterName=Flare, rendererClasses [&NiagaraSpriteRendererProperties]}
  {emitterName=Coin, rendererClasses [&NiagaraMeshRendererProperties]}
 ]
 userVariables [User.bIsFrontEndPreview {type=NiagaraBool}]
}
