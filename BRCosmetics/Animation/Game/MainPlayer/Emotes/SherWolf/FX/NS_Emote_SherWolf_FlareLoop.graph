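# Niagara system for the SherWolf emote's warm golden flare loop.
# Delayed flare and smoke bursts open over a five-partition flare emitter, while overhead beam sprites, drifting curl-noise pieces, sparkles and a glow mesh fill the space above.
# A flickering particle light adds illumination.
# Front-end preview flags, time-of-day scalars and platform switches adjust size, brightness and offsets, and curve-driven dynamic material parameters animate the smoke.
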NiagaraSystem /BRCosmetics/Animation/Game/MainPlayer/Emotes/SherWolf/FX/NS_Emote_SherWolf_FlareLoop
dependencies {
 usedDataInterfaces [
  &NiagaraDataInterfacePlatformSet
  &NiagaraDataInterfaceCurve
  &NiagaraDataInterfaceVector2DCurve
  &NiagaraDataInterfaceColorCurve
  &NiagaraDataInterfaceArrayFloat
  &NiagaraDataInterfaceArrayFloat2
  &NiagaraDataInterfaceArrayFloat3
  &NiagaraDataInterfaceMeshRendererInfo
 ]
 usedDynamicInputs [&RandomRangeVector2D]
 usedModules [&SystemState]
}
emitters {
 Beam {
  emitterUpdateScript [
   &EmitterState(Inactive Response=::NewEnumerator1, Life Cycle Mode=::NewEnumerator1, Loop Duration=1, UseLoopDelay=false)
   &SpawnRate(SpawnRate=2, Use Spawn Probability=false)
  ]
  particleSpawnScript [
   &InitializeParticle(
    Color=<rgba 1 0.3 0.032 1>
    Lifetime Max=3
    Lifetime Min=2
    Lifetime Mode=::NewEnumerator1
    Position Offset=<0 0 280>
    Sprite Size Max=<100 -180>
    Sprite Size Min=<120 -150>
    Sprite Size Mode=::NewEnumerator2
    UsePositionOffset=true
    Write Material Random=true
   )
   &AddVelocity(Velocity=<0 0 1>) {enabled=false}
   &AddVelocity(
    Cone Angle=0
    Cone Axis=<0 0 1>
    Module.UseConeFalloff=false
    Module.UseVelDistribution=false
    Override Seed=false
    Velocity Mode=::NewEnumerator2
    Velocity Speed=-5
   ) {module=AddVelocity001}
   &CameraOffset(Camera Offset Amount=-100)
  ]
  particleUpdateScript [
   &ParticleState()
   &ScaleSpriteSize(
    Non-Uniform Curve Sprite Scale=DataInterface() {propertyValues {
      XCurve {keys [{tangentMode=RCTM_User, time=0, value=1}, {tangentMode=RCTM_User, time=1, value=1}]}
      YCurve {keys [{tangentMode=RCTM_User, time=0, value=0.5}, {tangentMode=RCTM_User, time=1, value=2.5}]}
     }}
    Scale Sprite Size Mode=::NewEnumerator3
   )
   &SolveForcesAndVelocity(Clamp Velocity=false, Limit Acceleration=false)
   &ScaleColor(
    Scale Alpha=&FloatFromCurve(FloatCurve=DataInterface() {propertyValues {Curve {keys [<at 0 0 RCIM_Cubic>,<at 0.4 1 RCIM_Cubic>,<at 1 0 RCIM_Cubic>]}}})
    ScaleA=true
    ScaleRGB=true
   )
  ]
  rendererClasses [&NiagaraSpriteRendererProperties]
  renderers [{rendererClass=&NiagaraSpriteRendererProperties}]
 }
 Burst_Flare {
  emitterUpdateScript [
   &EmitterState(
    Inactive Response=::NewEnumerator1
    Life Cycle Mode=::NewEnumerator1
    Loop Behavior=::NewEnumerator1
    Loop Delay=0.8
    Loop Duration=3
    UseLoopDelay=true
   )
   &SpawnBurst_Instantaneous(Spawn Count=60, Use Spawn Probability=false, UseLoopCountLimit=false)
  ]
  particleSpawnScript [
   &InitializeParticle(
    Color=<rgba 3 1.5638 0.1236 1>
    Lifetime Max=0.7
    Lifetime Min=0.3
    Lifetime Mode=::NewEnumerator1
    Sprite Size Mode=::NewEnumerator1
    Uniform Sprite Size Max=8
    Uniform Sprite Size Min=6
    UsePositionOffset=false
    Write Material Random=true
   )
   &ShapeLocation(
    Cylinder Height=50
    MaskSpawnGroup=false
    Offset=<0 0 30>
    Offset Mode=::NewEnumerator0
    Override Seed=false
    Shape Primitive=::NewEnumerator1
    Surface Only=true
    Surface Only Band Thickness=1
    UseOwnerScale=true
   )
   &AddVelocity(Module.Use Falloff Distance=false, Origin Offset=<0 0 30>, Override Seed=false, Velocity Mode=::NewEnumerator1, Velocity Speed=200)
   &AddVelocity(Velocity=<0 0 20>) {module=AddVelocity001}
   &CameraOffset(Camera Offset Amount=30)
  ]
  particleUpdateScript [
   &ParticleState()
   &ScaleSpriteSize(
    Uniform Curve Scale=5
    Uniform Curve Sprite Scale=DataInterface() {propertyValues {Curve {keys [{tangentMode=RCTM_User, time=0, value=0}, <at 0.2 1>, <at 0.5 -0.45>, <at 0.7 0.75>, {tangentMode=RCTM_User, time=1, value=0}]}}}
   )
   &Drag(Drag=4, Use Linear Drag=true, Use Rotational Drag=true)
   /SetVariables(Particles.Color002=&Multiply_LinearColorByFloat(
     Linear Color=<rgba 2 1.171 0.0974 0.1>
     Float=&FNTimeOfDayScalar(Night Scale=0.7, Morning Scale=1, Evening Scale=1)
    )) {isSetParameters=true}
   &SolveForcesAndVelocity(Clamp Velocity=false, Limit Acceleration=false)
  ]
  rendererClasses [&NiagaraSpriteRendererProperties]
  renderers [
   {rendererClass=&NiagaraSpriteRendererProperties}
   {rendererClass=&NiagaraSpriteRendererProperties, rendererIndex=1}
  ]
 }
 Burst_Smoke {
  emitterUpdateScript [
   &EmitterState(
    Inactive Response=::NewEnumerator1
    Life Cycle Mode=::NewEnumerator1
    Loop Behavior=::NewEnumerator1
    Loop Delay=0.8
    Loop Duration=3
    UseLoopDelay=true
   )
   &SpawnBurst_Instantaneous(Spawn Count=20, Use Spawn Probability=false, UseLoopCountLimit=false)
  ]
  particleSpawnScript [
   &InitializeParticle(
    Color=<rgba 1 0.8519 0.4844 1>
    Lifetime Max=0.8
    Lifetime Min=0.5
    Lifetime Mode=::NewEnumerator1
    Sprite Rotation Mode=::NewEnumerator2
    Sprite Size Mode=::NewEnumerator1
    Uniform Sprite Size Max=100
    Uniform Sprite Size Min=60
    UsePositionOffset=false
    Write Material Random=true
   )
   &ShapeLocation(
    Cylinder Height=20
    Cylinder Radius=80
    MaskSpawnGroup=false
    Offset=<0 0 10>
    Offset Mode=::NewEnumerator0
    Override Seed=false
    Shape Primitive=::NewEnumerator1
    Surface Only=true
    Surface Only Band Thickness=1
    UseOwnerScale=true
   )
   &AddVelocity(Module.Use Falloff Distance=false, Origin Offset=<0 0 10>, Override Seed=false, Velocity Mode=::NewEnumerator1, Velocity Speed=120)
   &CameraOffset(Camera Offset Amount=&FN_FrontEndFloatModifier(Gameplay=30, Front End=10))
   &AddVelocity(
    Velocity=<0 0 80>
    Velocity Speed Scale=&FN_PickFloatDependOnPlatform(TargetPlatformActive=1, TargetPlatformNoActive=1.5)
   ) {module=AddVelocity001}
  ]
  particleUpdateScript [
   &ParticleState()
   &SpriteRotationRate(Rotation Rate=&Multiply_Float(
     A=&RandomRangeFloat(Minimum=-5, Maximum=5)
     B=&FloatFromCurve(
      FloatCurve=DataInterface() {propertyValues {Curve {keys [
          {arriveTangent=-0.9774, interpMode=RCIM_Cubic, leaveTangent=-0.9774, tangentMode=RCTM_User, time=0, value=1}
          {arriveTangent=-0.2496, interpMode=RCIM_Cubic, leaveTangent=-0.2496, tangentMode=RCTM_User, time=0.5, value=0.7}
         ]}}}
      CurveIndex=$Emitter.Age
      Scale Curve=20
     )
    ))
   &ScaleColor(
    Scale Alpha=&FloatFromCurve(
     FloatCurve=DataInterface() {propertyValues {Curve {keys [
         {arriveTangent=-3.8889, interpMode=RCIM_Cubic, time=0.1, value=0}
         {arriveTangent=0.1859, interpMode=RCIM_Cubic, leaveTangent=0.1859, tangentMode=RCTM_User, time=0.3, value=1}
         <at 1 0 RCIM_Cubic>
        ]}}}
     Scale Curve=1.2
    )
    Scale RGB=&VectorFromFloat(Value=&FloatFromCurve(FloatCurve=DataInterface() {propertyValues {Curve {keys [
          {arriveTangent=1.8145, interpMode=RCIM_Cubic, leaveTangent=1.0283, tangentMode=RCTM_Break, time=0, value=0.4}
          {arriveTangent=0.4198, interpMode=RCIM_Cubic, leaveTangent=0.4198, tangentMode=RCTM_User, time=0.4, value=0.825}
          {arriveTangent=0.1918, tangentMode=RCTM_User, time=1, value=1}
         ]}}}))
    ScaleA=true
    ScaleRGB=true
   )
   &Drag(Drag=2, Use Linear Drag=true, Use Rotational Drag=false)
   &ScaleSpriteSize(Scale Factor=&Vector2DFromFloat(Value=&FloatFromCurve(FloatCurve=DataInterface() {propertyValues {Curve {keys [
          {arriveTangent=1.9205, interpMode=RCIM_Cubic, leaveTangent=1.9205, tangentMode=RCTM_User, time=0, value=0.2}
          {arriveTangent=0.7125, interpMode=RCIM_Cubic, leaveTangent=0.7125, tangentMode=RCTM_User, time=0.3, value=0.75}
          {arriveTangent=0.1384, interpMode=RCIM_Cubic, leaveTangent=0.1384, tangentMode=RCTM_User, time=1, value=1}
         ]}}})))
   &DynamicMaterialParameters(
    Index 0 Param 1=&Add_Float(
     A=$Particles.RandomFloat
     B=&FloatFromCurve(
      FloatCurve=DataInterface() {propertyValues {Curve {keys [
          {arriveTangent=2.7533, interpMode=RCIM_Cubic, leaveTangent=1.8958, tangentMode=RCTM_User, time=0, value=1}
          {arriveTangent=0.8292, interpMode=RCIM_Cubic, leaveTangent=0.8292, tangentMode=RCTM_User, time=0.321, value=1.3453}
          {arriveTangent=0.0529, interpMode=RCIM_Cubic, leaveTangent=0.2021, tangentMode=RCTM_User, time=1, value=1.5}
         ]}}}
      CurveIndex=&Add_Float(A=&Multiply_Float(A=$Particles.Initial.MaterialRandom, B=0.5), B=$Particles.NormalizedAge)
     )
    )
    Index 0 Param 2=&FloatFromCurve(FloatCurve=DataInterface() {propertyValues {Curve {keys [
         {arriveTangent=0.0453, interpMode=RCIM_Cubic, tangentMode=RCTM_Break, time=0, value=-1}
         {arriveTangent=0.0604, interpMode=RCIM_Cubic, leaveTangent=0.0604, tangentMode=RCTM_User, time=1, value=-1}
        ]}}})
    Index 0 Param 3=&FloatFromCurve(
     FloatCurve=DataInterface() {propertyValues {Curve {keys [
         {arriveTangent=-1.5217, interpMode=RCIM_Cubic, leaveTangent=-1.5217, tangentMode=RCTM_User, time=0, value=0.2}
         {arriveTangent=0.1017, interpMode=RCIM_Cubic, leaveTangent=0.2723, tangentMode=RCTM_Break, time=0.2, value=0}
        ]}}}
     Scale Curve=0.1
    )
    Index 0 Param 4=&FloatFromCurve(FloatCurve=DataInterface() {propertyValues {Curve {keys [<at 0.5 1>,<at 1 0>]}}})
    Index 1 Param 1=&FloatFromCurve(
     FloatCurve=DataInterface() {propertyValues {Curve {keys [
         {interpMode=RCIM_Cubic, leaveTangent=4.7088, tangentMode=RCTM_Break, time=0, value=0}
         {arriveTangent=4.2257, interpMode=RCIM_Cubic, leaveTangent=4.2257, tangentMode=RCTM_User, time=0.25, value=1}
        ]}}}
     Scale Curve=&RandomRangeFloat(Minimum=1.5, Maximum=1.5)
    )
    Index 1 Param 2=&FloatFromCurve(
     FloatCurve=DataInterface() {propertyValues {Curve {keys [<at 0 0.2>, {leaveTangent=0.9066, time=1, value=2}]}}}
     Scale Curve=&RandomRangeFloat(Minimum=0.2, Maximum=0.3)
    )
    Index 1 Param 3=&FloatFromCurve(FloatCurve=DataInterface() {propertyValues {Curve {keys [
         {interpMode=RCIM_Cubic, leaveTangent=6.3917, tangentMode=RCTM_Break, time=0, value=0.2}
         {arriveTangent=1.7549, interpMode=RCIM_Cubic, leaveTangent=10.6767, tangentMode=RCTM_User, time=0.25, value=1}
        ]}}})
    Index 1 Param 4=&FloatFromCurve(
     FloatCurve=DataInterface() {propertyValues {Curve {keys [{arriveTangent=-0.136, interpMode=RCIM_Cubic, leaveTangent=-0.136, tangentMode=RCTM_User, time=0, value=1}, <at 1 0.95>]}}}
     Scale Curve=&RandomRangeFloat(Minimum=0.5, Maximum=1)
    )
    Index 2 Param 1=&FloatFromCurve(
     FloatCurve=DataInterface() {propertyValues {Curve {keys [<at 0 0.8>,<at 1 1>]}}}
     Scale Curve=&FloatFromCurve(FloatCurve=DataInterface() {propertyValues {Curve {keys [<at 1.85 1>,<at 2 1.2>]}}}, CurveIndex=$Emitter.Age)
    )
    Index 2 Param 2=&FloatFromCurve(
     FloatCurve=DataInterface() {propertyValues {Curve {keys [
         {arriveTangent=1.9056, interpMode=RCIM_Cubic, leaveTangent=1.9056, tangentMode=RCTM_User, time=0, value=0.2}
         {arriveTangent=0.0259, interpMode=RCIM_Cubic, leaveTangent=0.0259, tangentMode=RCTM_User, time=1, value=1.4}
        ]}}}
     Scale Curve=&RandomRangeFloat(Minimum=2, Maximum=0.7)
    )
    Index 2 Param 3=&FloatFromCurve(
     FloatCurve=DataInterface() {propertyValues {Curve {keys [
         {interpMode=RCIM_Cubic, leaveTangent=2.3064, tangentMode=RCTM_Break, time=0, value=1}
         {interpMode=RCIM_Cubic, leaveTangent=11.1233, tangentMode=RCTM_Break, time=1, value=2}
        ]}}}
     Scale Curve=2
    )
    Index 2 Param 4=&FloatFromCurve(
     FloatCurve=DataInterface() {propertyValues {Curve {keys [
         {arriveTangent=-0.6867, interpMode=RCIM_Cubic, leaveTangent=-0.6867, tangentMode=RCTM_User, time=0, value=1}
         {interpMode=RCIM_Cubic, leaveTangent=-1.2956, tangentMode=RCTM_Break, time=1, value=0.75}
        ]}}}
     Scale Curve=0.1
    )
    Module.Param0WriteEnabled=true
    Module.Param0WriteEnabled1=true
    Module.Param0WriteEnabled2=true
    Module.Param1WriteEnabled=false
    Module.Param1WriteEnabled1=true
    Module.Param1WriteEnabled2=true
    Module.Param2WriteEnabled=true
    Module.Param2WriteEnabled1=true
    Module.Param2WriteEnabled2=true
    Module.Param3WriteEnabled=true
    Module.Param3WriteEnabled1=true
    Module.Param3WriteEnabled2=true
    Write Parameter Index 1=true
    Write Parameter Index 2=true
   )
   &SolveForcesAndVelocity(Clamp Velocity=false, Limit Acceleration=false)
   /SetVariables(Particles.Color002=&MakeLinearColorFromVectorAndFloat(
     "Vector (RGB)"=&MakeVectorFromLinearColorRGB(Color=<rgba 1 0.8519 0.4844 1>)
     "Float (Alpha)"=&FloatFromCurve(
      FloatCurve=DataInterface() {propertyValues {Curve {keys [
          {arriveTangent=-3.8889, interpMode=RCIM_Cubic, time=0.1, value=0}
          {arriveTangent=0.1859, interpMode=RCIM_Cubic, leaveTangent=0.1859, tangentMode=RCTM_User, time=0.3, value=1}
          <at 1 0 RCIM_Cubic>
         ]}}}
      Scale Curve=0.35
     )
    )) {isSetParameters=true}
  ]
  rendererClasses [&NiagaraSpriteRendererProperties]
  renderers [
   {rendererClass=&NiagaraSpriteRendererProperties}
   {rendererClass=&NiagaraSpriteRendererProperties, rendererIndex=1}
  ]
 }
 Flare_loop001 {
  emitterUpdateScript [
   &EmitterState(
    Delay First Loop Only=true
    Inactive Response=::NewEnumerator1
    Life Cycle Mode=::NewEnumerator1
    Loop Delay=0.8
    Loop Duration=1
    UseLoopDelay=true
   )
   &SpawnRate(SpawnRate=50, Use Spawn Probability=false)
  ]
  particleSpawnScript [
   &PartitionParticles(Number Of Partitions=5)
   &InitializeParticle(
    Color=<rgba 2 1.0425 0.0824 1>
    Lifetime Max=1
    Lifetime Min=0.5
    Lifetime Mode=::NewEnumerator1
    Sprite Size Mode=::NewEnumerator2
    UsePositionOffset=false
    Write Material Random=true
   )
   &ShapeLocation(
    Hemisphere Distribution=&SelectVector2DFromArray(
     Direct Array Index=$Particles.PartitionParticles.Partition
     Vector 2D Selection Array=DataInterface() {propertyValues {FloatData [<1 0.5>, <1 0.5>, <1 1>, <1 0.5>, <1 0.5>]}}
    )
    MaskSpawnGroup=true
    Non Uniform Scale=&SelectVectorFromArray(
     Direct Array Index=$Particles.PartitionParticles.Partition
     Vector Selection Array=DataInterface() {propertyValues {FloatData [<1 1 0.8>, <1 1 1>, <0.9 0.9 2>, <1 1 1>, <1 1 0.8>]}}
    )
    Offset=&SelectVectorFromArray(
     Direct Array Index=$Particles.PartitionParticles.Partition
     Vector Selection Array=DataInterface() {propertyValues {FloatData [<0 -10 0>, <-60 50 0>, <50 20 10>, <75 40 0>, <0 -10 0>]}}
    )
    Offset Mode=::NewEnumerator0
    Override Seed=false
    Sphere Radius=&SelectFloatFromArray(
     Direct Array Index=$Particles.PartitionParticles.Partition
     Float Selection Array=DataInterface() {propertyValues {FloatData [65,15,20,15,65]}}
    )
    Sphere Surface Distribution=0.3
    UseOwnerScale=true
   )
   &CameraOffset(Camera Offset Amount=10)
  ]
  particleUpdateScript [
   &ParticleState()
   &ScaleSpriteSize(
    Uniform Curve Scale=&MakeCustomFloatFromBool(Bool=$User.bIsFrontEndPreview, True Float=4, False Float=5)
    Uniform Curve Sprite Scale=DataInterface() {propertyValues {Curve {keys [{tangentMode=RCTM_User, time=0, value=0}, <at 0.2 1>, <at 0.3 0.4>, {tangentMode=RCTM_User, time=1, value=0}]}}}
   )
   /SetVariables(Particles.Color002=<rgba 2 1.171 0.0974 0.03>) {isSetParameters=true}
  ]
  rendererClasses [&NiagaraSpriteRendererProperties]
  renderers [
   {rendererClass=&NiagaraSpriteRendererProperties}
   {rendererClass=&NiagaraSpriteRendererProperties, rendererIndex=1}
  ]
 }
 Light {
  emitterUpdateScript [
   &EmitterState(
    Delay First Loop Only=true
    Inactive Response=::NewEnumerator1
    Life Cycle Mode=::NewEnumerator1
    Loop Delay=1
    Loop Duration=7
    UseLoopDelay=true
   )
   &SpawnBurst_Instantaneous(Use Spawn Probability=false, UseLoopCountLimit=false)
  ]
  particleSpawnScript [
   &PartitionParticles() {enabled=false}
   &InitializeParticle(Lifetime=7, Mesh Scale Mode=::NewEnumerator4, UsePositionOffset=false, Write Material Random=true)
  ]
  particleUpdateScript [
   &ParticleState()
   &FN_Light_Attributes(
    Color=&ColorFromCurve(ColorCurve=DataInterface() {propertyValues {
       AlphaCurve {keys []}
       BlueCurve {keys [<at 0.1413 0.4193 RCIM_Cubic>]}
       GreenCurve {keys [<at 0.1413 1.9478 RCIM_Cubic>]}
       RedCurve {keys [<at 0.1413 3 RCIM_Cubic>]}
      }})
    Create Color Attribute=true
    Create Exponent Attribute=true
    Create Position Attribute=true
    Create Radius Attribute=true
    Enable Particle Light Enable=true
    Exponent=1.5
    Position=<0 50 50>
    Radius=&SelectFloatFromArray(Direct Array Index=$Particles.PartitionParticles.Partition, Float Selection Array=DataInterface() {propertyValues {FloatData [85,80]}})
    Scale Color=&FloatFromCurve(
     FloatCurve=DataInterface() {propertyValues {Curve {keys [
         {arriveTangent=13.8359, interpMode=RCIM_Cubic, leaveTangent=13.8359, tangentMode=RCTM_User, time=0.0412, value=0}
         {arriveTangent=0.0672, interpMode=RCIM_Cubic, leaveTangent=0.0672, tangentMode=RCTM_User, time=0.15, value=1}
         {arriveTangent=-0.4623, interpMode=RCIM_Cubic, leaveTangent=-0.4623, time=0.2, value=0.7987}
         {arriveTangent=0.2563, interpMode=RCIM_Cubic, leaveTangent=0.2563, time=0.2357, value=0.9604}
         {arriveTangent=-0.1575, interpMode=RCIM_Cubic, leaveTangent=-0.1575, time=0.3437, value=0.8355}
         {arriveTangent=-0.2189, interpMode=RCIM_Cubic, leaveTangent=-0.2189, time=0.4459, value=0.9273}
         {arriveTangent=-0.0069, interpMode=RCIM_Cubic, leaveTangent=-0.0069, time=0.5647, value=0.7871}
         {arriveTangent=0.0609, interpMode=RCIM_Cubic, leaveTangent=0.0609, time=0.6459, value=0.9259}
         {arriveTangent=-3.6443, interpMode=RCIM_Cubic, leaveTangent=-3.6443, time=0.7432, value=0.798}
         {arriveTangent=-0.1905, interpMode=RCIM_Cubic, leaveTangent=-0.1905, tangentMode=RCTM_User, time=0.9, value=0}
         <at 1 0>
        ]}}}
     Scale Curve=&Multiply_Float(
      A=&MakeCustomFloatFromBool(
       Bool=$User.bIsFrontEndPreview
       True Float=&SelectFloatFromArray(Direct Array Index=$Particles.PartitionParticles.Partition, Float Selection Array=DataInterface() {propertyValues {FloatData [2,2]}})
       False Float=&SelectFloatFromArray(Direct Array Index=$Particles.PartitionParticles.Partition, Float Selection Array=DataInterface() {propertyValues {FloatData [2.5,2]}})
      )
      B=&Multiply_Float(A=1, B=&FNTimeOfDayScalar(Night Scale=0.8, Morning Scale=0.8, Evening Scale=0.8))
     )
    )
   )
  ]
  rendererClasses [&NiagaraLightRendererProperties]
  renderers [{rendererClass=&NiagaraLightRendererProperties, rendererIndex=0}]
 }
 Piece_feeble {
  emitterUpdateScript [
   &EmitterState(Inactive Response=::NewEnumerator1, Life Cycle Mode=::NewEnumerator1, Loop Duration=1, UseLoopDelay=false)
   &SpawnRate(SpawnRate=3, Use Spawn Probability=false)
  ]
  particleSpawnScript [
   &InitializeParticle(
    Lifetime Max=5
    Lifetime Min=3
    Lifetime Mode=::NewEnumerator1
    Mesh Scale Max=<0.38 0.28 0.5>
    Mesh Scale Min=<0.05 0.05 0.05>
    Mesh Scale Mode=::NewEnumerator2
    Position Offset=<0 0 260>
    Sprite Rotation Mode=::NewEnumerator2
    Sprite Size Mode=::NewEnumerator1
    Uniform Sprite Size Max=15
    Uniform Sprite Size Min=8
    UsePositionOffset=true
    Write Material Random=true
   )
   &ShapeLocation(MaskSpawnGroup=false, Override Seed=false, Sphere Radius=40, UseOwnerScale=true)
   &InitialMeshOrientation(
    Mesh Orientation Mode=::NewEnumerator2
    Rotation=&RandomRangeVector(Minimum=<0 0 0>, Maximum=<1 1 1>)
    Use Rotation Vector=true
   )
   &AddVelocity(
    Cone Angle=55
    Cone Axis=<0 0 1>
    Module.UseConeFalloff=false
    Module.UseVelDistribution=false
    Override Seed=false
    Velocity Mode=::NewEnumerator2
    Velocity Speed=&RandomRangeFloat(Minimum=-15, Maximum=-25)
   ) {module=AddVelocity001}
  ]
  particleUpdateScript [
   &ParticleState()
   &UpdateMeshOrientation(
    Rotation Rate=&RandomRangeFloat(Minimum=0.5, Maximum=1.5)
    Rotation Vector=&RandomRangeVector(Minimum=<0.1 0.1 0.1>, Maximum=<0.5 0.5 0.5>)
   )
   &ScaleMeshSize(Scale Factor=&VectorFromFloat(Value=&FloatFromCurve(FloatCurve=DataInterface() {propertyValues {Curve {keys [<at 0 0>,<at 0.2 1>,<at 0.9 1>,<at 1 0>]}}})))
   &CurlNoiseForce(Bias Noise=false, Pan Noise=false)
   &GravityForce(Gravity=<0 0 -10>)
   &SolveForcesAndVelocity(Clamp Velocity=false, Limit Acceleration=false)
   /SetVariables(Particles.Color002=&MakeLinearColorFromVectorAndFloat(
     "Vector (RGB)"=&MakeVectorFromLinearColorRGB(Color=<rgba 2 1.0309 0.0923 1>)
     "Float (Alpha)"=&FloatFromCurve(
      FloatCurve=DataInterface() {propertyValues {Curve {keys [<at 0 0>,<at 0.2 1>,<at 0.8 1>,<at 1 0>]}}}
      Scale Curve=&MakeCustomFloatFromBool(
       Bool=$User.bIsFrontEndPreview
       True Float=0.04
       False Float=&RandomRangeFloat(Minimum=0.05, Maximum=0.1)
      )
     )
    )) {isSetParameters=true}
  ]
  rendererClasses [&NiagaraMeshRendererProperties, &NiagaraSpriteRendererProperties]
  renderers [
   {rendererClass=&NiagaraMeshRendererProperties, rendererIndex=0}
   {rendererClass=&NiagaraSpriteRendererProperties, rendererIndex=1}
  ]
 }
 Sparklys {
  emitterUpdateScript [
   &EmitterState(
    Inactive Response=::NewEnumerator1
    Life Cycle Mode=::NewEnumerator1
    Loop Duration=&RandomRangeFloat(Minimum=1, Maximum=0.5)
    UseLoopDelay=false
   )
   &SpawnRate(SpawnRate=4, Use Spawn Probability=false)
  ]
  particleSpawnScript [
   &InitializeParticle(
    Color=<rgba 2 2 2 1>
    Lifetime Max=0.6
    Lifetime Min=0.3
    Lifetime Mode=::NewEnumerator1
    Sprite Rotation Mode=::NewEnumerator2
    Sprite Size Mode=::NewEnumerator1
    Uniform Sprite Size Max=30
    Uniform Sprite Size Min=10
    UsePositionOffset=false
    Write Material Random=true
   )
   &ShapeLocation(
    MaskSpawnGroup=false
    Non Uniform Scale=<1 1 1.2>
    Offset=<0 0 150>
    Offset Mode=::NewEnumerator0
    Override Seed=false
    Shape Origin=<0 0 0>
    Sphere Radius=60
    UseOwnerScale=true
   )
   &CameraOffset(Camera Offset Amount=-100)
  ]
  particleUpdateScript [
   &ParticleState()
   &Drag(Drag=2.5, Use Linear Drag=true, Use Rotational Drag=false)
   &SolveForcesAndVelocity(Clamp Velocity=false, Limit Acceleration=false)
   &Color(Color=&ColorFromCurve(
     ColorCurve=DataInterface() {propertyValues {
       AlphaCurve {keys [<at 0 0.5 RCIM_Cubic>]}
       BlueCurve {keys [<at 0.25 0.1119 RCIM_Cubic>,<at 0.7 0.1186>]}
       GreenCurve {keys [<at 0.25 0.6423 RCIM_Cubic>,<at 0.7 0.4832>]}
       RedCurve {keys [<at 0.25 1 RCIM_Cubic>,<at 0.7 1>]}
      }}
     CurveIndex=$Emitter.NormalizedLoopAge
    )) {module=Color001}
   &ScaleColor(
    Color Value To Scale=$Particles.Color
    Scale Alpha=&FloatFromCurve(FloatCurve=DataInterface() {propertyValues {Curve {keys [
         {arriveTangent=-0.012, interpMode=RCIM_Cubic, leaveTangent=-0.012, tangentMode=RCTM_User, time=0, value=0}
         {arriveTangent=-1.4617, interpMode=RCIM_Cubic, leaveTangent=-1.4617, tangentMode=RCTM_User, time=0.3, value=1}
         {interpMode=RCIM_Cubic, tangentMode=RCTM_User, time=0.5, value=0.4}
         {interpMode=RCIM_Cubic, tangentMode=RCTM_User, time=0.7, value=0.6}
         {interpMode=RCIM_Cubic, tangentMode=RCTM_Break, time=1, value=0}
        ]}}})
    ScaleA=true
    ScaleRGB=true
   )
   /SetVariables(Particles.New Position:NiagaraPosition=&Add_Vector(
     A=Linked() {linkedVariable {name=Particles.Position, type=NiagaraPosition}}
     B=Linked() {linkedVariable {name=Engine.Owner.Position, type=NiagaraPosition}}
    )) {isSetParameters=true}
  ]
  rendererClasses [&NiagaraSpriteRendererProperties]
  renderers [{rendererClass=&NiagaraSpriteRendererProperties}]
 }
 light_feeble {
  emitterUpdateScript [
   &EmitterState(Inactive Response=::NewEnumerator1, Life Cycle Mode=::NewEnumerator1, UseLoopDelay=false)
   &SpawnBurst_Instantaneous(Use Spawn Probability=false, UseLoopCountLimit=false)
  ]
  particleSpawnScript [
   &InitializeParticle(
    Color=<rgba 1 0.63 0.0898 1>
    Lifetime=6
    Mesh Scale=<2.2 2.2 1.8>
    Mesh Scale Mode=::NewEnumerator4
    Position Offset=&MakeCustomVectorFromBool(Bool=$User.bIsFrontEndPreview, True Vector=<0 0 250>, False Vector=<0 0 330>)
    UsePositionOffset=true
    Write Material Random=true
   )
   &CameraOffset(Camera Offset Amount=&FN_PickFloatDependOnPlatform(TargetPlatformActive=-100, TargetPlatformNoActive=-150))
   /SetVariables(Particles.Size002:Vector3f=<10 18 10>) {isSetParameters=true}
   /SetVariables(Particles.Position002:NiagaraPosition=<0 -10 280>) {isSetParameters=true}
  ]
  particleUpdateScript [
   &ParticleState()
   &SolveForcesAndVelocity(Clamp Velocity=false, Limit Acceleration=false)
   &ScaleColor(
    Scale Alpha=&FloatFromCurve(
     FloatCurve=DataInterface() {propertyValues {Curve {keys [<at 0 0>, {arriveTangent=1.25, time=0.15, value=1}, {arriveTangent=-1.25, leaveTangent=-1.25, time=0.8, value=1}, <at 1 0>]}}}
     Scale Curve=&Multiply_Float(
      A=&MakeCustomFloatFromBool(Bool=$User.bIsFrontEndPreview, True Float=0.7, False Float=1)
      B=&Multiply_Float(
       A=&FNTimeOfDayScalar(Night Scale=0.8, Morning Scale=1, Evening Scale=1)
       B=&FN_PickFloatDependOnPlatform(TargetPlatformActive=1, TargetPlatformNoActive=0.6)
      )
     )
    )
    ScaleA=true
    ScaleRGB=true
   )
   &DynamicMaterialParameters(
    Index 0 Param 1=&MakeCustomFloatFromBool(Bool=$User.bIsFrontEndPreview, True Float=0.2, False Float=0.5)
    Index 0 Param 2=1
    Index 0 Param 3=1
    Index 0 Param 4=1
    Module.Param0WriteEnabled=true
    Module.Param1WriteEnabled=true
    Module.Param2WriteEnabled=true
    Module.Param3WriteEnabled=true
   )
  ]
  rendererClasses [&NiagaraMeshRendererProperties]
  renderers [{rendererClass=&NiagaraMeshRendererProperties, rendererIndex=0}]
 }
}
summary {
 systemName=NS_Emote_SherWolf_FlareLoop
 emitters [
  {emitterName=Burst_Smoke, rendererClasses [&NiagaraSpriteRendererProperties]}
  {emitterName=Burst_Flare, rendererClasses [&NiagaraSpriteRendererProperties]}
  {emitterName=Beam, rendererClasses [&NiagaraSpriteRendererProperties]}
  {emitterName=Sparklys, rendererClasses [&NiagaraSpriteRendererProperties]}
  {emitterName=Flare_loop001, rendererClasses [&NiagaraSpriteRendererProperties]}
  {emitterName=Piece_feeble, rendererClasses [&NiagaraMeshRendererProperties, &NiagaraSpriteRendererProperties]}
  {emitterName=light_feeble, rendererClasses [&NiagaraMeshRendererProperties]}
  {emitterName=Light, rendererClasses [&NiagaraLightRendererProperties]}
 ]
 userVariables [
  User.bIsFrontEnd {type=NiagaraBool}
  User.bIsFrontEndPreview {type=NiagaraBool}
 ]
}
