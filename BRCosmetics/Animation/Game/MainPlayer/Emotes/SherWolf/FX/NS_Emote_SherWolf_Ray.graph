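# Niagara effect for the SherWolf emote that shows a warm light ray with tumbling, drifting pieces.
# light002 bursts a mesh ray that fades in and out, scaling alpha by platform and time of day and setting camera offset by platform.
# Piece002 briefly spawns gold-tinted mesh and sprite particles above the origin that sink with curl noise.
# User.bIsFrontEndPreview adjusts ray height, opacity, a material parameter and piece alpha for front-end preview.
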NiagaraSystem /BRCosmetics/Animation/Game/MainPlayer/Emotes/SherWolf/FX/NS_Emote_SherWolf_Ray
dependencies {
 usedDataInterfaces [
  &NiagaraDataInterfaceCurve
  &NiagaraDataInterfaceMeshRendererInfo
  &NiagaraDataInterfacePlatformSet
 ]
 usedDynamicInputs []
 usedModules [&SystemState]
}
emitters {
 Piece002 {
  emitterUpdateScript [
   &EmitterState(Inactive Response=::NewEnumerator1, Life Cycle Mode=::NewEnumerator1, Loop Behavior=::NewEnumerator1, Loop Duration=8, UseLoopDelay=false)
   &SpawnRate(
    SpawnRate=&FloatFromCurve(FloatCurve=DataInterface() {propertyValues {Curve {keys [<at 0 1>,<at 1.5 1>,<at 2 0>,<at 8 0>]}}}, CurveIndex=$Emitter.Age, Scale Curve=15)
    Use Spawn Probability=false
   )
  ]
  particleSpawnScript [
   &InitializeParticle(
    Lifetime Max=4
    Lifetime Min=2
    Lifetime Mode=::NewEnumerator1
    Mesh Scale Max=<0.38 0.28 0.5>
    Mesh Scale Min=<0.05 0.05 0.05>
    Mesh Scale Mode=::NewEnumerator2
    Position Offset=<0 0 260>
    Sprite Rotation Mode=::NewEnumerator2
    Sprite Size Mode=::NewEnumerator1
    Uniform Sprite Size Max=15
    Uniform Sprite Size Min=8
    UsePositionOffset=true
    Write Material Random=true
   )
   &ShapeLocation(MaskSpawnGroup=false, Non Uniform Scale=<1 1 2>, Override Seed=false, Sphere Radius=40, UseOwnerScale=true)
   &InitialMeshOrientation(
    Mesh Orientation Mode=::NewEnumerator2
    Rotation=&RandomRangeVector(Minimum=<0 0 0>, Maximum=<1 1 1>)
    Use Rotation Vector=true
   )
   &AddVelocity(
    Cone Angle=50
    Cone Axis=<0 0 1>
    Module.UseConeFalloff=false
    Module.UseVelDistribution=false
    Override Seed=false
    Velocity Mode=::NewEnumerator2
    Velocity Speed=&RandomRangeFloat(Minimum=-15, Maximum=-25)
   ) {module=AddVelocity001}
  ]
  particleUpdateScript [
   &ParticleState()
   &UpdateMeshOrientation(
    Rotation Rate=&RandomRangeFloat(Minimum=1, Maximum=1.5)
    Rotation Vector=&RandomRangeVector(Minimum=<0.1 0.1 0.1>, Maximum=<0.5 0.5 0.5>)
   )
   &ScaleMeshSize(Scale Factor=&VectorFromFloat(Value=&FloatFromCurve(FloatCurve=DataInterface() {propertyValues {Curve {keys [<at 0 0>,<at 0.2 1>,<at 0.9 1>,<at 1 0>]}}})))
   &CurlNoiseForce(Bias Noise=false, Pan Noise=false)
   &GravityForce(Gravity=<0 0 -10>)
   &SolveForcesAndVelocity(Clamp Velocity=false, Limit Acceleration=false)
   /SetVariables(Particles.Color002=&MakeLinearColorFromVectorAndFloat(
     "Vector (RGB)"=&MakeVectorFromLinearColorRGB(Color=<rgba 2 1.0309 0.0923 1>)
     "Float (Alpha)"=&FloatFromCurve(
      FloatCurve=DataInterface() {propertyValues {Curve {keys [<at 0 0>,<at 0.2 1>,<at 0.8 1>,<at 1 0>]}}}
      Scale Curve=&MakeCustomFloatFromBool(
       Bool=$User.bIsFrontEndPreview
       True Float=0.04
       False Float=&RandomRangeFloat(Minimum=0.05, Maximum=0.1)
      )
     )
    )) {isSetParameters=true}
  ]
  rendererClasses [&NiagaraMeshRendererProperties, &NiagaraSpriteRendererProperties]
  renderers [
   {rendererClass=&NiagaraMeshRendererProperties, rendererIndex=0}
   {rendererClass=&NiagaraSpriteRendererProperties, rendererIndex=1}
  ]
 }
 light002 {
  emitterUpdateScript [
   &EmitterState(Inactive Response=::NewEnumerator1, Life Cycle Mode=::NewEnumerator1, Loop Behavior=::NewEnumerator1, Loop Duration=8, UseLoopDelay=false)
   &SpawnBurst_Instantaneous(Use Spawn Probability=false, UseLoopCountLimit=false)
  ]
  particleSpawnScript [
   &InitializeParticle(
    Color=<rgba 1 0.6 0.0614 1>
    Lifetime=4
    Mesh Scale=<2.2 2.2 1.8>
    Mesh Scale Mode=::NewEnumerator4
    Position Offset=&MakeCustomVectorFromBool(Bool=$User.bIsFrontEndPreview, True Vector=<0 0 250>, False Vector=<0 0 330>)
    UsePositionOffset=true
    Write Material Random=true
   )
   &CameraOffset(Camera Offset Amount=&FN_PickFloatDependOnPlatform(TargetPlatformActive=-100, TargetPlatformNoActive=-150))
   /SetVariables(Particles.Size002:Vector3f=<10 18 10>) {isSetParameters=true}
   /SetVariables(Particles.Position002:NiagaraPosition=<0 -10 280>) {isSetParameters=true}
  ]
  particleUpdateScript [
   &ParticleState()
   &SolveForcesAndVelocity(Clamp Velocity=false, Limit Acceleration=false)
   &ScaleColor(
    Scale Alpha=&FloatFromCurve(
     FloatCurve=DataInterface() {propertyValues {Curve {keys [
         <at 0 0>
         {arriveTangent=1.25, time=0.15, value=1}
         {arriveTangent=-0.9805, interpMode=RCIM_Cubic, leaveTangent=-0.9805, tangentMode=RCTM_User, time=0.5, value=1}
         {arriveTangent=-2, interpMode=RCIM_Cubic, leaveTangent=-2, time=0.7069, value=0.3516}
         <at 1 0 RCIM_Cubic>
        ]}}}
     Scale Curve=&Multiply_Float(
      A=&MakeCustomFloatFromBool(Bool=$User.bIsFrontEndPreview, True Float=0.6, False Float=1)
      B=&Multiply_Float(
       A=&FNTimeOfDayScalar(Night Scale=0.8, Morning Scale=1, Evening Scale=1)
       B=&FN_PickFloatDependOnPlatform(TargetPlatformActive=1, TargetPlatformNoActive=0.6)
      )
     )
    )
    ScaleA=true
    ScaleRGB=true
   )
   &DynamicMaterialParameters(
    Index 0 Param 1=&MakeCustomFloatFromBool(Bool=$User.bIsFrontEndPreview, True Float=0.1, False Float=0.3)
    Index 0 Param 2=1
    Index 0 Param 3=1
    Index 0 Param 4=1
    Module.Param0WriteEnabled=true
    Module.Param1WriteEnabled=true
    Module.Param2WriteEnabled=true
    Module.Param3WriteEnabled=true
   )
  ]
  rendererClasses [&NiagaraMeshRendererProperties]
  renderers [{rendererClass=&NiagaraMeshRendererProperties, rendererIndex=0}]
 }
}
summary {
 systemName=NS_Emote_SherWolf_Ray
 emitters [
  {emitterName=Piece002, rendererClasses [&NiagaraMeshRendererProperties, &NiagaraSpriteRendererProperties]}
  {emitterName=light002, rendererClasses [&NiagaraMeshRendererProperties]}
 ]
 userVariables [
  User.bIsFrontEnd {type=NiagaraBool}
  User.bIsFrontEndPreview {type=NiagaraBool}
 ]
}
