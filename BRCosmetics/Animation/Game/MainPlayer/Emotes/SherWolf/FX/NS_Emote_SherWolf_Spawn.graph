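# Niagara spawn effect for the SherWolf emote, producing golden sparks and a soft glow.
# Flare_High (GPU) and lower-rate Flare_Low share small sphere-spawned sprites with curl noise, drag and randomized brightness; front-end preview trims their spawn rate.
# The "golden" emitter adds larger, faint SubUV sprites that inherit capped velocity, with alpha scaled by time of day.
# All three emitters follow source movement horizontally through InheritSourceMovement.
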NiagaraSystem /BRCosmetics/Animation/Game/MainPlayer/Emotes/SherWolf/FX/NS_Emote_SherWolf_Spawn
dependencies {
 usedDataInterfaces [
  &NiagaraDataInterfaceCurve
  &NiagaraDataInterfaceSpriteRendererInfo
  &NiagaraDataInterfaceSkeletalMesh
 ]
 usedDynamicInputs []
 usedModules [
  &self:SystemSpawnScript.SystemScriptSource.SystemScriptGraph.NiagaraNodeAssignment_0.SetVariables_#1
  &SystemState
 ]
}
emitters {
 Flare_High {
  simTarget=GPUComputeSim
  emitterUpdateScript [
   &EmitterState(Life Cycle Mode=::NewEnumerator1, Loop Behavior=::NewEnumerator1, Loop Duration=10, UseLoopDelay=false)
   &SpawnRate(
    SpawnRate=&Multiply_Float(
     A=&RandomRangeFloat(Minimum=300, Maximum=800)
     B=&MakeCustomFloatFromBool(Bool=$User.bIsFrontEndPreview, True Float=0.8, False Float=1)
    )
    Use Spawn Probability=false
   )
  ]
  particleSpawnScript [
   &InitializeParticle(
    Color=<rgba 2 1.0425 0.0824 1>
    Lifetime Max=0.8
    Lifetime Min=0.2
    Lifetime Mode=::NewEnumerator1
    Sprite Size Mode=::NewEnumerator1
    Uniform Sprite Size Max=5
    Uniform Sprite Size Min=1
    UsePositionOffset=false
    Write Material Random=true
   )
   &ShapeLocation(MaskSpawnGroup=false, Override Seed=false, Sphere Radius=10, UseOwnerScale=true)
   &AddVelocity(Velocity=<0 10 0>)
  ]
  particleUpdateScript [
   &ParticleState()
   &ScaleSpriteSize(
    Uniform Curve Scale=&Multiply_Float(
     A=&RandomRangeFloat(Minimum=3, Maximum=7)
     B=&RandomRangeFloat(Minimum=0.5, Maximum=1, Evaluation Type=::NewEnumerator1)
    )
    Uniform Curve Sprite Scale=DataInterface() {propertyValues {Curve {keys [
        <at 0 0 RCIM_Cubic>
        {arriveTangent=-2.9089, interpMode=RCIM_Cubic, leaveTangent=-2.9089, tangentMode=RCTM_User, time=0.2, value=1}
        <at 1 0 RCIM_Cubic>
       ]}}}
   )
   &Drag(Use Linear Drag=true, Use Rotational Drag=true)
   &CurlNoiseForce(Bias Noise=false, Noise Strength=100, Pan Noise=false)
   &InheritSourceMovement(Applied Position Scale=<1 1 0>, ApplyForce=false, ApplyPosition=true, ApplyPositionFalloff=false)
   &SolveForcesAndVelocity(Clamp Velocity=false, Limit Acceleration=false)
   /SetVariables(Particles.Color002=<rgba 2 1.171 0.0974 0.03>) {isSetParameters=true, module=SetVariables_#2}
   &ScaleColor(
    Scale RGB=&Multiply_VectorByFloat(Vector=<1 1 1>, Float=&RandomRangeFloat(Minimum=1, Maximum=4))
    ScaleA=true
    ScaleRGB=true
   )
  ]
  rendererClasses [&NiagaraSpriteRendererProperties]
  renderers [
   {rendererClass=&NiagaraSpriteRendererProperties}
   {rendererClass=&NiagaraSpriteRendererProperties, rendererIndex=1}
  ]
 }
 Flare_Low {
  emitterUpdateScript [
   &EmitterState(Life Cycle Mode=::NewEnumerator1, Loop Behavior=::NewEnumerator1, Loop Duration=10, UseLoopDelay=false)
   &SpawnRate(
    SpawnRate=&Multiply_Float(
     A=&RandomRangeFloat(Minimum=100, Maximum=200)
     B=&MakeCustomFloatFromBool(Bool=$User.bIsFrontEndPreview, True Float=0.8, False Float=1)
    )
    Use Spawn Probability=false
   )
  ]
  particleSpawnScript [
   &InitializeParticle(
    Color=<rgba 2 1.0425 0.0824 1>
    Lifetime Max=0.8
    Lifetime Min=0.2
    Lifetime Mode=::NewEnumerator1
    Sprite Size Mode=::NewEnumerator1
    Uniform Sprite Size Max=5
    Uniform Sprite Size Min=1
    UsePositionOffset=false
    Write Material Random=true
   )
   &ShapeLocation(MaskSpawnGroup=false, Override Seed=false, Sphere Radius=10, UseOwnerScale=true)
   &AddVelocity(Velocity=<0 10 0>)
  ]
  particleUpdateScript [
   &ParticleState()
   &ScaleSpriteSize(
    Uniform Curve Scale=&Multiply_Float(
     A=&RandomRangeFloat(Minimum=3, Maximum=7)
     B=&RandomRangeFloat(Minimum=0.5, Maximum=1, Evaluation Type=::NewEnumerator1)
    )
    Uniform Curve Sprite Scale=DataInterface() {propertyValues {Curve {keys [
        <at 0 0 RCIM_Cubic>
        {arriveTangent=-2.9089, interpMode=RCIM_Cubic, leaveTangent=-2.9089, tangentMode=RCTM_User, time=0.2, value=1}
        <at 1 0 RCIM_Cubic>
       ]}}}
   )
   &Drag(Use Linear Drag=true, Use Rotational Drag=true)
   &CurlNoiseForce(Bias Noise=false, Noise Strength=100, Pan Noise=false)
   &InheritSourceMovement(Applied Position Scale=<1 1 0>, ApplyForce=false, ApplyPosition=true, ApplyPositionFalloff=false)
   &SolveForcesAndVelocity(Clamp Velocity=false, Limit Acceleration=false)
   /SetVariables(Particles.Color002=<rgba 2 1.171 0.0974 0.03>) {isSetParameters=true, module=SetVariables_#2}
   &ScaleColor(
    Scale RGB=&Multiply_VectorByFloat(Vector=<1 1 1>, Float=&RandomRangeFloat(Minimum=1, Maximum=4))
    ScaleA=true
    ScaleRGB=true
   )
  ]
  rendererClasses [&NiagaraSpriteRendererProperties]
  renderers [
   {rendererClass=&NiagaraSpriteRendererProperties}
   {rendererClass=&NiagaraSpriteRendererProperties, rendererIndex=1}
  ]
 }
 golden {
  emitterUpdateScript [
   &EmitterState(Inactive Response=::NewEnumerator1, Life Cycle Mode=::NewEnumerator1, Loop Behavior=::NewEnumerator1, Loop Duration=10, UseLoopDelay=false)
   &SpawnRate(SpawnRate=&RandomRangeFloat(Minimum=50, Maximum=70), Use Spawn Probability=false)
  ]
  particleSpawnScript [
   &InitializeParticle(
    Color=<rgba 1 0.7743 0.1514 0.1>
    Lifetime Max=1
    Lifetime Mode=::NewEnumerator1
    Sprite Rotation Mode=::NewEnumerator2
    Sprite Size Mode=::NewEnumerator1
    Uniform Sprite Size Max=45
    Uniform Sprite Size Min=25
    UsePositionOffset=false
    Write Material Random=true
   )
   &ShapeLocation(MaskSpawnGroup=false, Override Seed=false, Sphere Radius=8, UseOwnerScale=true)
   &AddVelocity(Velocity=<0 10 0>) {enabled=false}
   &InheritVelocity(Inherited Velocity Speed Limit=30, Speed Limit=true, UseSpeedThreshold=false)
  ]
  particleUpdateScript [
   &ParticleState()
   &ScaleSpriteSize(Uniform Curve Sprite Scale=DataInterface() {propertyValues {Curve {keys [
        {arriveTangent=-0.1201, interpMode=RCIM_Cubic, leaveTangent=-0.1201, tangentMode=RCTM_User, time=0, value=1}
        {arriveTangent=-0.6994, interpMode=RCIM_Cubic, leaveTangent=-0.6994, tangentMode=RCTM_User, time=1, value=0.6}
       ]}}})
   &CurlNoiseForce(Bias Noise=false, Noise Strength=100, Pan Noise=false) {enabled=false}
   &InheritSourceMovement(Applied Position Scale=<1 1 0>, ApplyForce=false, ApplyPosition=true, ApplyPositionFalloff=false)
   &SolveForcesAndVelocity(Clamp Velocity=false, Limit Acceleration=false)
   &SubUVAnimation(
    End Frame Range Override=8
    "Sprite Renderer "=DataInterface() {propertyValues {SpriteRenderer=&self:golden_0.NiagaraSpriteRendererProperties_2}}
    Start Frame Range Override=8
    SubUV Animation Mode:ENiagara_SubUVLookupModeV2=::NewEnumerator2
    UseEndFrame=true
    UseStartFrame=true
   )
   &ScaleColor(
    Scale Alpha=&FloatFromCurve(
     FloatCurve=DataInterface() {propertyValues {Curve {keys [
         <at 0 0 RCIM_Cubic>
         {arriveTangent=-4.0512, interpMode=RCIM_Cubic, leaveTangent=-4.0512, tangentMode=RCTM_User, time=0.1, value=1}
         {arriveTangent=-0.4436, interpMode=RCIM_Cubic, leaveTangent=-0.4436, tangentMode=RCTM_User, time=0.3, value=0.1}
         <at 1 0 RCIM_Cubic>
        ]}}}
     Scale Curve=&Multiply_Float(
      A=&Multiply_Float(A=0.2, B=&FNTimeOfDayScalar(Night Scale=0.1, Morning Scale=1, Evening Scale=1))
      B=1
     )
    )
    ScaleA=true
    ScaleRGB=true
   )
  ]
  rendererClasses [&NiagaraSpriteRendererProperties]
  renderers [{rendererClass=&NiagaraSpriteRendererProperties}]
 }
}
summary {
 systemName=NS_Emote_SherWolf_Spawn
 emitters [
  {emitterName=Flare_High, simTarget=GPUComputeSim, rendererClasses [&NiagaraSpriteRendererProperties]}
  {emitterName=golden, rendererClasses [&NiagaraSpriteRendererProperties]}
  {emitterName=Flare_Low, rendererClasses [&NiagaraSpriteRendererProperties]}
 ]
 userVariables [
  User.bIsFrontEnd {type=NiagaraBool}
  User.bIsFrontEndPreview {type=NiagaraBool}
 ]
}
